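# Grasslands ambient GPU Niagara system with fireflies and wind-blown leaves, plus a disabled local fog emitter.
# Fireflies spawn by time of day, sample the landscape and drift on curl noise and wind while avoiding distance-field surfaces.
# Leaves are projected onto the depth buffer, culled when invalid, then tumble under gravity and wind, with gusts raising spawn rate.
# Both active emitters use GBuffer hue-range sampling at spawn, with per-partition ranges for leaves.
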NiagaraSystem /Game/Effects/Fort_Effects/Effects/Environment/Grasslands/NS_GrassLands_BaseFX
dependencies {
 usedDataInterfaces [
  &NiagaraDataInterfaceCurve
  &NiagaraDataInterfaceArrayColor
  &NiagaraDataInterfaceArrayFloat
  &NiagaraDataInterfaceCollisionQuery
  &NiagaraDataInterfaceGBuffer
  &NiagaraDataInterfaceVector2DCurve
  &NiagaraDataInterfaceLandscape
  &NiagaraDataInterfaceColorCurve
  &NiagaraDataInterfaceCamera
 ]
 usedDynamicInputs [&SimulationPosition]
 usedModules [
  &self:SystemSpawnScript.SystemScriptSource.SystemScriptGraph.NiagaraNodeAssignment_0.SetVariables_#1
  &SystemState
 ]
}
emitters {
 Firefiles_Epic {
  simTarget=GPUComputeSim
  emitterSpawnScript [/SetVariables(Emitter.Camera Position:NiagaraPosition=<-77365 12953 3500>) {isSetParameters=true, module=SetVariables_#2}]
  emitterUpdateScript [
   &EmitterState()
   &SpawnRate(
    Spawn Probability=&FloatFromCurve(
     FloatCurve=DataInterface() {propertyValues {Curve {keys [<at 0 0 RCIM_Constant>,<at 0.25 0 RCIM_Constant>,<at 0.5 0 RCIM_Constant>,<at 0.75 1 RCIM_Constant>,<at 1 1 RCIM_Constant>,<at 1 1 RCIM_Constant>]}}}
     CurveIndex=&MakeFloatFromLinearColor(LinearColor=$NPC.FortniteNPC.FortniteActiveTimeOfDay, Channel:ENiagaraLinearColor_Channels=::NewEnumerator3)
    )
    SpawnRate=12.5
    Use Spawn Probability=true
   )
  ]
  particleSpawnScript [
   &InitializeParticle(
    Color=&ColorFromCurve(ColorCurve=DataInterface() {propertyValues {
       AlphaCurve {keys [<at 0 1 RCIM_Cubic>]}
       BlueCurve {keys [<at 0 0 RCIM_Cubic>,<at 1 0 RCIM_Cubic>]}
       GreenCurve {keys [<at 0 1 RCIM_Cubic>,<at 1 1 RCIM_Cubic>]}
       RedCurve {keys [<at 0 0.4298 RCIM_Cubic>,<at 1 0.8694 RCIM_Cubic>]}
      }})
    Lifetime=5
    Mesh Scale Mode=::NewEnumerator1
    Mesh Uniform Scale Max=0.6
    Mesh Uniform Scale Min=0.36
    Sprite Size Mode=::NewEnumerator3
    Uniform Sprite Size=75
    UsePositionOffset=false
    Write Material Random=true
   )
   &ShapeLocation(MaskSpawnGroup=false, Override Seed=false, Sphere Radius=7500, UseOwnerScale=true)
   &self:SampleLandscape_01() {module=SampleLandscape_01}
   &KillParticlesInVolume(Invert Volume=true, Sphere Radius=12500)
   &CameraOffset(Camera Offset Amount=1)
   /SetVariables(Particles.Spline Scale:Vector3f=<1 1 1>) {isSetParameters=true}
   &AccelerationForce(Acceleration=&Multiply_VectorByFloat(Vector=$Output.SampleLandscape_01.LandscapeNormal, Float=300)) {module=AccelerationForce001}
   &ApplyInitialForces()
   /Game/Effects/Niagara/Modules/Spawn/FN_Spawn_GBuffer_Color_WithinRangeV2(HueMax=101, HueMin=98)
  ]
  particleUpdateScript [
   &ParticleState()
   &ScaleMeshSize(Scale Factor=&VectorFromFloat(Value=1)) {module=ScaleMeshSize001}
   &ScaleMeshSize(Scale Factor=&VectorFromFloat(Value=&FloatFromCurve(FloatCurve=DataInterface() {propertyValues {Curve {keys [
          <at 0 0 RCIM_Cubic>
          {arriveTangent=1.1765, leaveTangent=1.1765, time=0.1, value=1}
          {interpMode=RCIM_Cubic, tangentMode=RCTM_Break, time=0.8, value=1}
          {arriveTangent=-23.4104, interpMode=RCIM_Cubic, tangentMode=RCTM_Break, time=1, value=0}
         ]}}})))
   &ScaleSpriteSize(Uniform Curve Sprite Scale=DataInterface() {propertyValues {Curve {keys [
        <at 0 0 RCIM_Cubic>
        {arriveTangent=1.1765, leaveTangent=1.1765, time=0.1, value=1}
        {interpMode=RCIM_Cubic, tangentMode=RCTM_Break, time=0.8, value=1}
        {arriveTangent=-23.4104, interpMode=RCIM_Cubic, tangentMode=RCTM_Break, time=1, value=0}
       ]}}})
   &CurlNoiseForce(Bias Noise=false, Noise Frequency=5, Noise Strength=750, Pan Noise=false)
   &WindForce(Wind Speed Scale=&Lerp_Float(A=0.1, B=5, Alpha=$NPC.FortniteNPC.WindGustsSystem_Active))
   &Drag(Drag=10, Use Linear Drag=true, Use Rotational Drag=true)
   &AccelerationForce(Acceleration=<50 0 300>)
   &PointAttractionForce(Attraction Radius=150, Attraction Strength=-500, Kill Within Radius=false, Module.Use Falloff=true, UseAttractorOffset=false)
   /Niagara/Modules/Math/AvoidDistanceFieldSurfaces_GPU(Avoid Nearest Surface=true, Avoid Oncoming Surfaces=true, Nearest Surface Avoidance Strength=5)
   &SolveForcesAndVelocity(Clamp Velocity=false, Limit Acceleration=false)
   &DynamicMaterialParameters(
    Index 0 Param 4=&Lerp_Float(A=10, B=25, Alpha=&Sine(Normalized Angle=$Particles.Age, Period=1, Scale=1, Bias=0.5))
    Index 1 Param 1=1
    Module.Index0Param0Write=true
    Module.Index0Param1Write=true
    Module.Index0Param2Write=true
    Module.Index0Param3Write=true
    Module.Index1Param0Write=true
    Module.Index1Param1Write=true
    Module.Index1Param2Write=true
    Module.Index1Param3Write=true
    Write Parameter Index 1=true
   )
   &ScaleColor(
    Linear Color Curve=DataInterface() {propertyValues {AlphaCurve {keys [<at 0.8906 0.45>,<at 1 0>]}, BlueCurve {keys [<at 0 0.5>]}, GreenCurve {keys [<at 0 0.5>]}, RedCurve {keys [<at 0 0.5>]}}}
    Scale Mode=::NewEnumerator2
   )
   &InterpolateOverTime(Rate Of Change=5, Target Position Value:NiagaraPosition=$Particles.Position, Variable Type=::NewEnumerator6)
   /SetVariables(Particles.Vector 4:Vector4f=&MakeVector4(
     X=&MakeFloatFromVector(
      Vector=&Subtract_Position(A=$Particles.InterpolateOverTime.InterpolateOverTime_Position.Moving Average, B=$Particles.Position)
      Channel=::NewEnumerator0
     )
     Y=&MakeFloatFromVector(
      Vector=&Subtract_Position(A=$Particles.InterpolateOverTime.InterpolateOverTime_Position.Moving Average, B=$Particles.Position)
      Channel=::NewEnumerator1
     )
     Z=&MakeFloatFromVector(
      Vector=&Subtract_Position(A=$Particles.InterpolateOverTime.InterpolateOverTime_Position.Moving Average, B=$Particles.Position)
      Channel=::NewEnumerator2
     )
     W=1
    )) {isSetParameters=true}
   &ScaleColor(Scale Mode=::NewEnumerator2) {enabled=false, module=ScaleColor001}
   &ScaleMeshSize(Scale Factor=&VectorFromFloat(Value=&Lerp_Float(
      A=0
      B=1
      Alpha=&RemapRange(
       Input Value To Remap=&VectorLength(Vector=&SubtractVectorFromPosition(
         Position=&ReturnCameraPosition(Coordinate Space:ENiagaraCoordinateSpace=::Simulation, Camera Query:NiagaraDataInterfaceCamera=DataInterface())
         Vector=<0 0 0>
        ))
       Input Min=0
       Input Max=1
       Output Min=0
       Output Max=1
      )
     ))) {module=ScaleMeshSize002}
   &FN_Sprite_Attributes(
    Camera Offset=20
    Color=&HueShiftLinearColor(Hue=0, Saturation=0.25, Value=10, Linear Color=$Particles.Color)
    Create Camera Offset Attribute=true
    Create Color Attribute=true
    Create Position Attribute=true
    Create SpriteSize Attribute=true
    Position:NiagaraPosition=&AddVectorToPosition(
     Position=$Particles.Position
     Vector=&TransformVector(Vector=<0 0 0>, Source Space=::World, Destination Space=::Local, Apply Scale=true)
    )
    SpriteSize=&Multiply_Vector2D_ByFloat(
     Float=&Lerp_Float(
      A=0
      B=1
      Alpha=&RemapRange(
       Input Value To Remap=&VectorLength(Vector=&SubtractVectorFromPosition(
         Position=/Niagara/DynamicInputs/Helpers/SimulationPosition()
         Vector=Linked() {linkedVariable {name=Particles.Position, type=NiagaraPosition}}
        ))
       Input Min=0
       Input Max=2500
       Output Min=0
       Output Max=1
      )
     )
     Vector2D=<40 40>
    )
    Write to Module Namespace=true
   )
   /SetVariables(Particles.Vector 4:Vector4f=<1 1 0 1>) {isSetParameters=true}
  ]
  rendererClasses [&NiagaraMeshRendererProperties, &NiagaraSpriteRendererProperties]
  renderers [
   {rendererClass=&NiagaraMeshRendererProperties, rendererIndex=0}
   {rendererClass=&NiagaraSpriteRendererProperties, rendererIndex=1}
   {rendererClass=&NiagaraMeshRendererProperties, rendererIndex=2}
   {rendererClass=&NiagaraSpriteRendererProperties, rendererIndex=3}
   {rendererClass=&NiagaraSpriteRendererProperties, rendererIndex=4}
   {rendererClass=&NiagaraSpriteRendererProperties, rendererIndex=5}
  ]
 }
 LeavesInAir {
  simTarget=GPUComputeSim
  emitterUpdateScript [
   /SetVariables(
    Emitter.Emitter_HeightOffset:NiagaraFloat=$System.System_HeightOffset
    Emitter.Overall_SpeedMult=0.2
    Emitter.SpawnRate_Curve:NiagaraFloat=&FloatFromCurve(
     FloatCurve=DataInterface() {propertyValues {Curve {keys [{arriveTangent=17.5837, time=0, value=0}, <at 0.1 1>, <at 0.2 0.0925>, <at 0.6 0.1525>, <at 0.7 0>, <at 1 0>]}}}
     CurveIndex=$Emitter.NormalizedLoopAge
     Scale Curve=100
    )
    Emitter.SpawnSpeed_Curve:NiagaraFloat=&FloatFromCurve(
     FloatCurve=DataInterface() {propertyValues {Curve {keys [
         {arriveTangent=30.2685, interpMode=RCIM_Cubic, leaveTangent=30.2685, tangentMode=RCTM_User, time=0.0009, value=0.0215}
         {arriveTangent=-0.0619, interpMode=RCIM_Cubic, leaveTangent=-0.0619, tangentMode=RCTM_User, time=0.0989, value=0.3515}
         <at 0.297 0.0032>
         <at 0.3192 0.3332>
         <at 0.3525 0.0398>
         <at 0.6171 0.3148>
         <at 0.9939 0.3698>
        ]}}}
     CurveIndex=&Multiply_Float(A=$Emitter.NormalizedLoopAge, B=&RandomRangeFloat(Minimum=1, Maximum=5))
     Scale Curve=&Multiply_Float(
      A=&Lerp_Float(A=5, B=10, Alpha=&UniformRangedFloat(Minimum=0.25, Maximum=1))
      B=1
     )
    )
    Emitter.SpriteRotRate:NiagaraFloat=&FloatFromCurve(
     FloatCurve=DataInterface() {propertyValues {Curve {keys [
         <at 0.0041 0.1315>
         <at 0.1088 1.9664>
         {arriveTangent=-3.3757, interpMode=RCIM_Cubic, leaveTangent=-3.3757, tangentMode=RCTM_User, time=0.3609, value=0.6391}
         {arriveTangent=-1, interpMode=RCIM_Cubic, leaveTangent=-1, tangentMode=RCTM_User, time=0.5614, value=0.4386}
         <at 0.7086 1.6472>
         <at 1 0>
        ]}}}
     CurveIndex=$Emitter.NormalizedLoopAge
     Scale Curve=500
    )
   ) {isSetParameters=true}
   &EmitterState(Life Cycle Mode=::NewEnumerator1, Loop Duration=0.05, UseLoopDelay=false)
   &SpawnBurst_Instantaneous(
    Spawn Count=&RandomRangeInt(Minimum=10, Maximum=50)
    Use Spawn Probability=false
    UseLoopCountLimit=false
   )
   &SpawnRate(
    SpawnRate=&Multiply_Float(A=1750, B=&Lerp_Float(A=0.025, B=1, Alpha=$NPC.FortniteNPC.WindGustsSystem_Active))
    Use Spawn Probability=false
   )
  ]
  particleSpawnScript [
   &PartitionParticles(Number Of Partitions=4)
   /SetVariables(Particles.HeightOffset:NiagaraFloat=&RandomRangeFloat(Minimum=17500, Maximum=3000)) {isSetParameters=true}
   &InitializeParticle(
    Color=&SelectLinearColorFromArray(
     Array Sampling Mode=::NewEnumerator1
     Direct Array Index=$Particles.PartitionParticles.Partition
     Color Selection Array=DataInterface() {propertyValues {ColorData [<rgba 0.0203 0.0887 0.0056 1>, <rgba 0.2423 0.0823 0.1845 1>]}}
     Random Seed=?
    )
    Lifetime=&UniformRangedFloat(Minimum=3, Maximum=6)
    Mass Mode=::NewEnumerator1
    Mesh Scale Mode=::NewEnumerator3
    Mesh Uniform Scale=&RandomRangeFloat(
     Minimum=&SelectFloatFromArray(Direct Array Index=$Particles.PartitionParticles.Partition, Float Selection Array=DataInterface() {propertyValues {FloatData [0.25,0.0125]}})
     Maximum=&SelectFloatFromArray(Direct Array Index=$Particles.PartitionParticles.Partition, Float Selection Array=DataInterface() {propertyValues {FloatData [0.45,0.035]}})
    )
    Sprite Rotation Angle=&UniformRangedFloat(Minimum=-180, Maximum=180)
    Sprite Rotation Mode=::NewEnumerator0
    Sprite Size=&Vector2DFromFloat(Value=&UniformRangedFloat(Minimum=100, Maximum=150))
    Sprite Size Mode=::NewEnumerator4
    UsePositionOffset=false
    Write Material Random=true
   ) {module=InitializeParticle001}
   &AddVelocity(Velocity=&Multiply_VectorByFloat(
     Vector=&Multiply_VectorByFloat(
      Vector=&Multiply_VectorByFloat(Vector=&UniformRangedVector(Minimum=<30 50 -250>, Maximum=<80 150 500>), Float=$Emitter.SpawnSpeed_Curve)
      Float=$Emitter.Overall_SpeedMult
     )
     Float=1
    ))
   /Niagara/Modules/Spawn/Location/SphereLocation(
    MaskSpawnGroup=false
    Module.Override Randomness=false
    Non Uniform Scale=&RandomRangeVector(Minimum=<-1 -1 -1>, Maximum=<1 1 1>)
    Offset=<0 0 900>
    Sphere Radius=&Multiply_Float(
     A=7500
     B=&FloatFromCurve(FloatCurve=DataInterface() {propertyValues {Curve {keys [<at 0 0.1>,<at 0.157 0.9198>]}}}, CurveIndex=$System.NormalizedLoopAge)
    )
    Surface Only=false
   ) {module=SphereLocation001}
   /Niagara/Modules/Math/MoveToNearestDistanceFieldSurface_GPU(Kill by Distance=500, bKillByDistance=true) {enabled=false}
   /SetVariables(Particles.InitialPos:NiagaraPosition=$Particles.Position) {isSetParameters=true}
   /Niagara/Modules/Math/PlaceParticlesOnDepthBuffer_GPU(Max Z Depth=10000, Module.Use Max Z Depth=true)
   &KillParticles(Kill Particles=&SetBoolByFloatComparison(
     A=&VectorLength(Vector=&Subtract_Vector(A=$Particles.InitialPos, B=Linked() {linkedVariable {name=Particles.Position, type=NiagaraPosition}}))
     B=1500
    ))
   &KillParticles(Kill Particles=&Boolean_NotOperation(Bool=$Output.PlaceParticlesOnDepthBuffer_GPU.ValidSample)) {module=KillParticles001}
   &KillParticles(Kill Particles=&SetBoolByFloatComparison(
     A=&MakeFloatFromVector(Vector=Linked() {linkedVariable {name=Particles.Position, type=NiagaraPosition}}, Channel=::NewEnumerator2)
     B=&MakeFloatFromVector(
      Vector=&AddVectorToPosition(Position=&SimulationPosition(), Vector=<0 0 180>)
      Channel=::NewEnumerator2
     )
     Comparison Type=::NewEnumerator6
    )) {module=KillParticles002}
   &FN_Spawn_GBuffer_Color_WithinRangeV2(
    HueMax:NiagaraFloat=&SelectFloatFromArray(Direct Array Index=$Particles.PartitionParticles.Partition, Float Selection Array=DataInterface() {propertyValues {FloatData [140,322]}})
    HueMin:NiagaraFloat=&SelectFloatFromArray(Direct Array Index=$Particles.PartitionParticles.Partition, Float Selection Array=DataInterface() {propertyValues {FloatData [105,305]}})
   )
   /Game/Effects/Niagara/Modules/Spawn/FN_Spawn_GBuffer_Color() {enabled=false}
   &CurlNoiseForce(
    Noise Frequency=25
    Noise Strength=&Multiply_Float(
     A=250
     B=&Multiply_Float(A=&Multiply_Float(A=1, B=$Emitter.Overall_SpeedMult), B=1)
    )
    Pan Noise=false
   ) {module=CurlNoiseForce001}
   &ApplyInitialForces()
   &CameraOffset(Camera Offset Amount=-200)
   /SetVariables(Particles.VisibilityTag=$Particles.PartitionParticles.Partition) {isSetParameters=true}
  ]
  particleUpdateScript [
   /SetVariables(Particles.ParticleUpdate_RotRate:NiagaraFloat=&FloatFromCurve(
     FloatCurve=DataInterface() {propertyValues {Curve {keys [
         <at 0.0041 0.1315>
         <at 0.1088 1.9664>
         {arriveTangent=-3.3757, interpMode=RCIM_Cubic, leaveTangent=-3.3757, tangentMode=RCTM_User, time=0.3609, value=0.6391}
         {arriveTangent=-1, interpMode=RCIM_Cubic, leaveTangent=-1, tangentMode=RCTM_User, time=0.5614, value=0.4386}
         <at 0.7086 1.6472>
         {arriveTangent=-1.004, interpMode=RCIM_Cubic, leaveTangent=-1.004, tangentMode=RCTM_User, time=0.8311, value=0.9545}
         {arriveTangent=-5.6531, interpMode=RCIM_Cubic, leaveTangent=-5.6531, tangentMode=RCTM_User, time=0.8987, value=1.4563}
         <at 0.9384 0.3483>
         <at 1 0>
        ]}}}
     Scale Curve=5
    )) {isSetParameters=true}
   &ParticleState()
   &UpdateMeshOrientation(Rotation Rate=&FloatFromCurve(
     FloatCurve=DataInterface() {propertyValues {Curve {keys [<at 250 1>,<at 500 4>]}}}
     CurveIndex=&MakeFloatFromVector(Vector=$Particles.Velocity, Channel=::NewEnumerator0)
    ))
   &ScaleMeshSize(Scale Factor=&Multiply_VectorByFloat(
     Vector=<1 1 1>
     Float=&FloatFromCurve(FloatCurve=DataInterface() {propertyValues {Curve {keys [
          {interpMode=RCIM_Cubic, tangentMode=RCTM_User, time=0, value=0}
          {arriveTangent=2, leaveTangent=2, time=0.1, value=1}
          <at 0.5 1>
          {arriveTangent=-2, leaveTangent=-2, time=0.8, value=1}
          {tangentMode=RCTM_User, time=1, value=0}
         ]}}})
    ))
   /Niagara/Modules/Update/SubUV/SubUVAnimation(Number Of Frames=64)
   &ScaleSpriteSize(Scale Factor=&Vector2DFromCurve(
     Vector2Curve=DataInterface() {propertyValues {
       XCurve {keys [
         {arriveTangent=5.9728, interpMode=RCIM_Cubic, leaveTangent=5.9728, tangentMode=RCTM_User, time=-0.0064, value=0.0304}
         {arriveTangent=-0.1, interpMode=RCIM_Cubic, leaveTangent=-0.1, tangentMode=RCTM_User, time=0.3581, value=0.9069}
         {interpMode=RCIM_Cubic, tangentMode=RCTM_User, time=1, value=0}
        ]}
       YCurve {keys [
         {arriveTangent=6.2715, interpMode=RCIM_Cubic, leaveTangent=6.2715, tangentMode=RCTM_User, time=-0.0064, value=0.0122}
         {arriveTangent=-0.1001, interpMode=RCIM_Cubic, leaveTangent=-0.1001, tangentMode=RCTM_User, time=0.3581, value=0.9359}
         {interpMode=RCIM_Cubic, tangentMode=RCTM_User, time=0.9988, value=0}
        ]}
      }}
     ScaleCurve=<1 1>
    )) {enabled=false}
   &ScaleSpriteSize(Scale Factor=&Multiply_Vector2(
     A=<1 1>
     B=&Vector2DFromFloat(Value=&Multiply_Float(
       A=&FloatFromCurve(FloatCurve=DataInterface() {propertyValues {Curve {keys [
            {interpMode=RCIM_Cubic, tangentMode=RCTM_User, time=-0.0061, value=0.489}
            {arriveTangent=0.3677, interpMode=RCIM_Cubic, leaveTangent=0.3677, tangentMode=RCTM_User, time=0.2892, value=0.9107}
            {arriveTangent=-1.2812, interpMode=RCIM_Cubic, leaveTangent=-1.2812, tangentMode=RCTM_User, time=0.9809, value=0.9275}
            <at 1 0>
           ]}}})
       B=&FloatFromCurve(
        FloatCurve=DataInterface() {propertyValues {Curve {keys [
            {interpMode=RCIM_Cubic, tangentMode=RCTM_User, time=0, value=0.0005}
            {arriveTangent=0.0074, interpMode=RCIM_Cubic, leaveTangent=0.0074, tangentMode=RCTM_User, time=0.15, value=1.0236}
            {arriveTangent=-0.0952, interpMode=RCIM_Cubic, leaveTangent=-0.0952, tangentMode=RCTM_User, time=0.85, value=1}
            <at 0.9 0>
           ]}}}
        CurveIndex=&NormalizeDistanceRange(
         Start Position:NiagaraPosition=Linked() {linkedVariable {name=Emitter.CameraPostion, type=Vector3f}}
         End Position:NiagaraPosition=$Particles.Position
         Distance:NiagaraFloat=10000
         Invert Normalized Range:NiagaraBool=false
        )
        Scale Curve=0.85
       )
      ))
    )) {module=ScaleSpriteSize001}
   &Drag(Use Linear Drag=true, Use Rotational Drag=false)
   &CurlNoiseForce(
    Curl Noise Cone Mask Angle=50
    Mask Curl Noise=true
    Noise Frequency=&RandomRangeFloat(Minimum=5, Maximum=25)
    Noise Strength=&Multiply_Float(A=$Emitter.SpawnSpeed_Curve, B=500)
    Pan Noise=true
    Pan Noise Field=<0 0 1>
    Random Seed=80
   )
   &ScaleColor(
    Scale Alpha=&FloatFromCurve(
     FloatCurve=DataInterface() {propertyValues {Curve {keys [
         {interpMode=RCIM_Cubic, tangentMode=RCTM_User, time=0, value=0}
         {arriveTangent=-0.0948, interpMode=RCIM_Cubic, leaveTangent=-0.0948, tangentMode=RCTM_User, time=0.1, value=1}
         <at 0.942 0.9825>
         <at 1 0>
        ]}}}
     Scale Curve=&FloatFromCurve(FloatCurve=DataInterface() {propertyValues {Curve {keys [
          {interpMode=RCIM_Cubic, tangentMode=RCTM_User, time=0, value=0}
          {interpMode=RCIM_Cubic, tangentMode=RCTM_User, time=0.2, value=5}
          {tangentMode=RCTM_User, time=1, value=0}
         ]}}})
    )
    ScaleA=true
    ScaleRGB=true
   )
   /Niagara/Modules/Update/SubUV/V2/SubUVAnimation(
    Start Frame:NiagaraInt32=&RandomRangeInt(Minimum=0, Maximum=32)
    SubUV Animation Mode:ENiagara_SubUVLookupMode=::NewEnumerator1
    SubUV Loop Count:NiagaraFloat=&SelectFloatFromArray(
     Array Sampling Mode=::NewEnumerator0
     Direct Array Index=?
     Float Selection Array=DataInterface() {propertyValues {FloatData [2,4]}}
     Random Seed=0
    )
   ) {module=SubUVAnimation001}
   &DynamicMaterialParameters(
    Index 0 Param 3=1
    Module.Param0WriteEnabled=true
    Module.Param1WriteEnabled=true
    Module.Param2WriteEnabled=true
    Module.Param3WriteEnabled=true
   )
   &WindForce(Wind Speed=<250 0 150>) {enabled=false}
   &GravityForce(Gravity=&Multiply_VectorByFloat(Vector=<0 0 -980>, Float=&RandomRangeFloat(Minimum=1, Maximum=2)))
   &WindForce(
    Friction=2
    Ground Distance Min / Max:Vector2f=<0 100>
    Ground Elevation:NiagaraFloat=&MakeFloatFromVector(Vector=Linked() {linkedVariable {name=Engine.Emitter.SimulationPosition, type=NiagaraPosition}}, Channel=::NewEnumerator2)
    Scale Min / Max:Vector2f=<0.5 1>
    Speed Min / Max:Vector2f=<0 150>
    Turbulence Scale Mode:ENiagaraWindTurbulenceContributionMode=::NewEnumerator3
    Use Ground Mask=true
    Wind Speed=<250 0 -100>
    Wind Speed Scale=&Lerp_Float(A=0.25, B=1, Alpha=$NPC.FortniteNPC.WindGustsSystem_Active)
   ) {module=WindForce001}
   &AerodynamicDrag(Aerodynamic Drag=0.5, Use Linear Drag=true, Use Rotational Drag=true)
   &SolveForcesAndVelocity(Clamp Velocity=false, Limit Acceleration=false)
   &AlignSpriteToMeshOrientation(Mesh Orientation Relative Sprite Facing Vector:Vector3f=<0 0 0>, SetAlignment=true, SetFacing=true)
  ]
  rendererClasses [&NiagaraMeshRendererProperties, &NiagaraSpriteRendererProperties]
  renderers [
   {rendererClass=&NiagaraMeshRendererProperties, rendererIndex=0}
   {rendererClass=&NiagaraSpriteRendererProperties, rendererIndex=1}
   {rendererClass=&NiagaraMeshRendererProperties, rendererIndex=2}
  ]
 }
 Local_Fog {
  bEnabled=false
  simTarget=GPUComputeSim
  emitterSpawnScript [/SetVariables(Emitter.Camera Position:NiagaraPosition=<-77365 12953 3500>) {enabled=false, isSetParameters=true, module=SetVariables_#2}]
  emitterUpdateScript [
   &EmitterState(Life Cycle Mode=::NewEnumerator1, Loop Duration=0.5, UseLoopDelay=false)
   &SpawnRate(SpawnRate=35, Use Spawn Probability=true)
  ]
  particleSpawnScript [
   &InitializeParticle(
    Color=<rgba 5 5 5 0.0063>
    Lifetime Max=6
    Lifetime Min=4
    Lifetime Mode=::NewEnumerator1
    Mesh Scale Mode=::NewEnumerator1
    Mesh Uniform Scale Max=1
    Mesh Uniform Scale Min=0.5
    Sprite Size Mode=::NewEnumerator1
    Uniform Sprite Size Max=1500
    Uniform Sprite Size Min=1000
    UsePositionOffset=false
    Write Material Random=true
   )
   &ShapeLocation(MaskSpawnGroup=false, Override Seed=false, Sphere Radius=6500, UseOwnerScale=true)
   &ShapeLocation(MaskSpawnGroup=false, Override Seed=false, Sphere Radius=1500, UseOwnerScale=true) {enabled=false, module=ShapeLocation001}
   &self:SampleLandscape(Offset:NiagaraPosition=<0 0 1>) {module=SampleLandscape}
   &KillParticlesInVolume(Invert Volume=true, Sphere Radius=7500)
  ]
  particleUpdateScript [
   &ParticleState()
   &ScaleColor(
    Linear Color Curve=DataInterface() {propertyValues {
      AlphaCurve {keys [<at 0 0>,<at 0.4945 0.5>,<at 1 0>]}
      BlueCurve {keys [<at 0 0>,<at 0.4952 1>,<at 1 0>]}
      GreenCurve {keys [<at 0 0>,<at 0.4952 1>,<at 1 0>]}
      RedCurve {keys [<at 0 0>,<at 0.4952 1>,<at 1 0>]}
     }}
    Scale Mode=::NewEnumerator2
   )
   &ScaleColor(
    Scale Mode=::NewEnumerator1
    Scale RGBA=&MakeVector4_FromLinearColor(LinearColor=&FNTimeOfDayColor(
      Night Color=<rgba 0.0503 0.061 0.1396 1>
      Morning Color=<rgba 0.1104 0.0789 0.0789 1>
      Day Color=<rgba 0.0896 0.0896 0.0896 1>
      Evening Color=<rgba 0.0688 0.0465 0.0512 1>
     ))
    ScaleRGBA=true
   ) {module=ScaleColor001}
  ]
  rendererClasses [&NiagaraSpriteRendererProperties]
  renderers [{rendererClass=&NiagaraSpriteRendererProperties}]
 }
}
summary {
 systemName=NS_GrassLands_BaseFX
 emitters [
  {
   emitterName=LeavesInAir
   simTarget=GPUComputeSim
   rendererClasses [&NiagaraMeshRendererProperties, &NiagaraSpriteRendererProperties]
  }
  {bEnabled=false, emitterName=Local_Fog, simTarget=GPUComputeSim, rendererClasses [&NiagaraSpriteRendererProperties]}
  {
   emitterName=Firefiles_Epic
   simTarget=GPUComputeSim
   rendererClasses [&NiagaraMeshRendererProperties, &NiagaraSpriteRendererProperties]
  }
 ]
 userVariables []
}
